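# Preview animation Blueprint for the Oblong vehicle body skeleton that poses the front wheels as if steered.
# Starting from the mesh-space reference pose, two chained Transform (Modify) Bone nodes additively apply a fixed 20-degree yaw to the front-left and front-right steering bones, then convert the result to local space.
# The pose is static: there are no variables, and the event graph holds only unconnected default nodes.
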AnimBlueprint /VehicleCosmetics/Bodies/Oblong/Animation/Body_Oblong_Preview_AnimBP
target_skeleton=/VehicleCosmetics/Bodies/Oblong/Meshes/SK_VEH_Oblong_Skeleton
graphs [
 AnimGraph {nodes [
   n1 = "Transform (Modify) Bone - Bone: wheel_steering_fr_l"(
    ComponentPose=Mesh Space Ref Pose!() -> (ComponentPose:Component Space Pose Link Structure) {is=title AnimGraphNode_MeshRefPose}
    Alpha=1
    Translation="0,0,0"
    Rotation="0,20,0"
    Scale="1,1,1"
   ) {rotationMode=BMM_Additive, alphaBoolBlend {blendOption=Linear}, boneToModify {boneName=wheel_steering_fr_l}, rotation {yaw=20}}
   n2 = "Transform (Modify) Bone - Bone: wheel_steering_fr_r"(ComponentPose=@n1, Alpha=1, Translation="0,0,0", Rotation="0,20,0", Scale="1,1,1") {rotationMode=BMM_Additive, alphaBoolBlend {blendOption=Linear}, boneToModify {boneName=wheel_steering_fr_r}, rotation {yaw=20}}
   Output Pose(Result=Component To Local(ComponentPose=@n2)) {layerGroup=None}
  ]}
 EventGraph {nodes [
   TryGetPawnOwner()
   Event BlueprintUpdateAnimation()
  ]}
]
variables []
