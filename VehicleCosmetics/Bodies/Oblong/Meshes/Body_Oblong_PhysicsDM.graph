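# Physics asset for the Oblong vehicle cosmetic body, giving its mesh a simplified physics shape.
# A single root body holds one box, root_box, longer than it is wide or tall and offset along positive X and upward from the bone origin.
# It has no constraints, so the asset defines no articulated or jointed parts.
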PhysicsAsset /VehicleCosmetics/Bodies/Oblong/Meshes/Body_Oblong_PhysicsDM
bodies {root [{shapeType=Box, extentX=444, extentY=216, extentZ=116, shapeName=root_box, center=<39.8612 0 89.0019>}]}
constraints []
