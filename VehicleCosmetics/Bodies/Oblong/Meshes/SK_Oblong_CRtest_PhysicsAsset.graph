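# Physics asset for the Oblong vehicle body's CRtest skeletal mesh, giving chassis, glass, steering and wheel bones collision bodies and joints.
# Each bone has one capsule, with only a tiny one on frame.
# Constraints share uniform swing and twist limits: body and four knuckles join frame, cabin parts join body, and wheel discs and brake pads follow knuckles, via steering bones at the front.
# Only the rear-left control arm has a body.
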PhysicsAsset /VehicleCosmetics/Bodies/Oblong/Meshes/SK_Oblong_CRtest_PhysicsAsset
bodies {
 body [{shapeType=Capsule, length=262.2575, radius=123.2271, center=<46.3777 0.0885 10.7319>, rotation=<rot 0 -90.1127 -85.5551>}]
 brakepad_bk_l [{shapeType=Capsule, length=27.2613, radius=18.5371, center=<0.2759 1.5983 5.4776>, rotation=<rot 0 -91.1767 -86.5364>}]
 brakepad_bk_r [{shapeType=Capsule, length=27.2613, radius=18.5371, center=<0.2763 -1.5985 5.4773>, rotation=<rot 0 -88.8233 -86.5364>}]
 brakepad_fr_l [{shapeType=Capsule, length=20.0298, radius=6.4527, center=<-19.9081 0.1819 -1.7412>, rotation=<rot 0.0001 -90 -4.9972>}]
 brakepad_fr_r [{shapeType=Capsule, length=20.0298, radius=6.4527, center=<-19.9081 -0.1822 -1.7419>, rotation=<rot -0.0001 -90 -4.9972>}]
 ctrlArm_bk_l [{shapeType=Capsule, length=26.7253, radius=14.2859, center=<-3.8358 -10.9819 0.813>, rotation=<rot 0 76.9841 -89.3903>}]
 frame [{shapeType=Capsule, length=0.505, radius=0.505, rotation=<rot 0 -90 0>}]
 glass_c [{shapeType=Capsule, length=86.1627, radius=83.4733, center=<-52.317 -2.3538 14.6206>, rotation=<rot 16.0957 -139.7666 -90>}]
 glass_l [{shapeType=Capsule, length=100.5035, radius=15.6681, center=<8.6888 9.504 17.2146>, rotation=<rot 0 -96.5578 -83.8089>}]
 glass_r [{shapeType=Capsule, length=100.5035, radius=15.6681, center=<8.6888 -9.504 17.2146>, rotation=<rot 0 -83.4422 -83.8089>}]
 steering_knuckle_bk_l [{shapeType=Capsule, length=25.6494, radius=23.828, center=<-1.8275 6.5228 -1.7812>, rotation=<rot 18.4971 -90 -7.4724>}]
 steering_knuckle_bk_r [{shapeType=Capsule, length=25.6494, radius=23.828, center=<-1.827 -6.523 -1.7816>, rotation=<rot -18.4971 -90 -7.4724>}]
 steering_knuckle_fr_l [{shapeType=Capsule, length=23.1549, radius=22.2642, center=<2.2409 10.6742 -1.8632>, rotation=<rot 16.8538 -90 -3.851>}]
 steering_knuckle_fr_r [{shapeType=Capsule, length=23.1549, radius=22.2642, center=<2.2409 -10.6744 -1.8637>, rotation=<rot -16.8537 -90 -3.8511>}]
 steering_wheel [{shapeType=Capsule, length=17.5777, radius=13.7962, center=<3.0916 -0.0298 1.3299>, rotation=<rot 89.485 80.8211 170.8214>}]
 wheel_disc_bk_l [{shapeType=Capsule, length=22.6974, radius=22.6974, center=<-0.0002 0.1195 0.0002>, rotation=<rot 0 -90 23.3544>}]
 wheel_disc_bk_r [{shapeType=Capsule, length=22.6974, radius=22.6974, center=<0.0003 -0.1198 -0.0002>, rotation=<rot 0 -90 -66.6453>}]
 wheel_disc_fr_l [{shapeType=Capsule, length=22.6974, radius=22.6974, center=<0 0.1197 -0.0001>, rotation=<rot 0.0001 -90 -66.6461>}]
 wheel_disc_fr_r [{shapeType=Capsule, length=22.6974, radius=22.6974, center=<0 -0.12 -0.0004>, rotation=<rot -0.0001 -90 -66.6458>}]
 wheel_steering_fr_l [{shapeType=Capsule, length=21.2511, radius=15.14, center=<-2.1973 8.405 -2.5057>, rotation=<rot 2.5384 -90 -10.8747>}]
 wheel_steering_fr_r [{shapeType=Capsule, length=21.2511, radius=15.14, center=<-2.1973 -8.405 -2.5061>, rotation=<rot -2.5384 -90 -10.8747>}]
}
constraints [
 {bone1Name=body, bone2Name=frame, swing1LimitDegrees=45, swing2LimitDegrees=45, twistLimitDegrees=45}
 {bone1Name=steering_wheel, bone2Name=body, swing1LimitDegrees=45, swing2LimitDegrees=45, twistLimitDegrees=45}
 {bone1Name=glass_c, bone2Name=body, swing1LimitDegrees=45, swing2LimitDegrees=45, twistLimitDegrees=45}
 {bone1Name=glass_l, bone2Name=body, swing1LimitDegrees=45, swing2LimitDegrees=45, twistLimitDegrees=45}
 {bone1Name=glass_r, bone2Name=body, swing1LimitDegrees=45, swing2LimitDegrees=45, twistLimitDegrees=45}
 {bone1Name=ctrlArm_bk_l, bone2Name=body, swing1LimitDegrees=45, swing2LimitDegrees=45, twistLimitDegrees=45}
 {bone1Name=steering_knuckle_fr_l, bone2Name=frame, swing1LimitDegrees=45, swing2LimitDegrees=45, twistLimitDegrees=45}
 {bone1Name=wheel_steering_fr_l, bone2Name=steering_knuckle_fr_l, swing1LimitDegrees=45, swing2LimitDegrees=45, twistLimitDegrees=45}
 {bone1Name=wheel_disc_fr_l, bone2Name=wheel_steering_fr_l, swing1LimitDegrees=45, swing2LimitDegrees=45, twistLimitDegrees=45}
 {bone1Name=brakepad_fr_l, bone2Name=wheel_steering_fr_l, swing1LimitDegrees=45, swing2LimitDegrees=45, twistLimitDegrees=45}
 {bone1Name=steering_knuckle_fr_r, bone2Name=frame, swing1LimitDegrees=45, swing2LimitDegrees=45, twistLimitDegrees=45}
 {bone1Name=wheel_steering_fr_r, bone2Name=steering_knuckle_fr_r, swing1LimitDegrees=45, swing2LimitDegrees=45, twistLimitDegrees=45}
 {bone1Name=wheel_disc_fr_r, bone2Name=wheel_steering_fr_r, swing1LimitDegrees=45, swing2LimitDegrees=45, twistLimitDegrees=45}
 {bone1Name=brakepad_fr_r, bone2Name=wheel_steering_fr_r, swing1LimitDegrees=45, swing2LimitDegrees=45, twistLimitDegrees=45}
 {bone1Name=steering_knuckle_bk_l, bone2Name=frame, swing1LimitDegrees=45, swing2LimitDegrees=45, twistLimitDegrees=45}
 {bone1Name=wheel_disc_bk_l, bone2Name=steering_knuckle_bk_l, swing1LimitDegrees=45, swing2LimitDegrees=45, twistLimitDegrees=45}
 {bone1Name=brakepad_bk_l, bone2Name=steering_knuckle_bk_l, swing1LimitDegrees=45, swing2LimitDegrees=45, twistLimitDegrees=45}
 {bone1Name=steering_knuckle_bk_r, bone2Name=frame, swing1LimitDegrees=45, swing2LimitDegrees=45, twistLimitDegrees=45}
 {bone1Name=wheel_disc_bk_r, bone2Name=steering_knuckle_bk_r, swing1LimitDegrees=45, swing2LimitDegrees=45, twistLimitDegrees=45}
 {bone1Name=brakepad_bk_r, bone2Name=steering_knuckle_bk_r, swing1LimitDegrees=45, swing2LimitDegrees=45, twistLimitDegrees=45}
]
